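# Physics asset for the Ramirez female medium head mesh, providing capsule bodies for a dynamic hair chain and the head and upper body.
# Four tapering dyn_hair_05 capsules form a chain constrained to the head, each joint carrying a much tighter twist limit than swing limit.
# Larger capsules cover faceAttach, head, neck_01 and spine_05, linked by constraints with free swing and twist.
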PhysicsAsset /Game/Characters/Player/Female/Medium/Heads/F_MED_HIS_Ramirez_Head_01/Mesh/F_MED_HIS_Ramirez_Head_01_Physics
bodies {
 dyn_hair_05 [{shapeType=Capsule, length=2.25, radius=3.5, center=<0.6218 0 0.2692>, rotation=<rot 66.5908 -180 -180>}]
 dyn_hair_05_a [{shapeType=Capsule, length=3, radius=2.75, center=<1.1885 0.0001 0>, rotation=<rot 90 0 0>}]
 dyn_hair_05_b [{shapeType=Capsule, length=2.75, radius=2.5, center=<0.9382 0 0>, rotation=<rot 90 0 0>}]
 dyn_hair_05_c [{shapeType=Capsule, length=2.5, radius=2, center=<1.0002 0 0>, rotation=<rot 90 0 0>}]
 faceAttach [{shapeType=Capsule, length=7.3926, radius=6.4904, center=<1.7578 7.4997 2.1388>, rotation=<rot 0 102.1037 -46.2044>}]
 head [{shapeType=Capsule, length=7.5921, radius=6.6484, center=<0.1705 0.1239 -0.432>, rotation=<rot 0 79.873 -89.9998>}]
 neck_01 [{shapeType=Capsule, length=4.5, radius=4.505, center=<0 0.2774 -0.0001>, rotation=<rot 0 -93.282 -89.9999>}]
 spine_05 [{shapeType=Capsule, length=9.75, radius=14, center=<0.0777 0.5063 0>}]
}
constraints [
 {
  bone1Name=faceAttach
  bone2Name=head
  swing1LimitDegrees=45
  swing1Motion=Free
  swing2LimitDegrees=45
  swing2Motion=Free
  twistLimitDegrees=45
  twistMotion=Free
 }
 {bone1Name=dyn_hair_05, bone2Name=head, swing1LimitDegrees=45, swing2LimitDegrees=45, twistLimitDegrees=5}
 {
  bone1Name=head
  bone2Name=neck_01
  swing1LimitDegrees=45
  swing1Motion=Free
  swing2LimitDegrees=45
  swing2Motion=Free
  twistLimitDegrees=45
  twistMotion=Free
 }
 {
  bone1Name=neck_01
  bone2Name=spine_05
  swing1LimitDegrees=45
  swing1Motion=Free
  swing2LimitDegrees=45
  swing2Motion=Free
  twistLimitDegrees=45
  twistMotion=Free
 }
 {bone1Name=dyn_hair_05_a, bone2Name=dyn_hair_05, swing1LimitDegrees=45, swing2LimitDegrees=45, twistLimitDegrees=5}
 {bone1Name=dyn_hair_05_b, bone2Name=dyn_hair_05_a, swing1LimitDegrees=45, swing2LimitDegrees=45, twistLimitDegrees=5}
 {bone1Name=dyn_hair_05_c, bone2Name=dyn_hair_05_b, swing1LimitDegrees=45, swing2LimitDegrees=45, twistLimitDegrees=5}
]
