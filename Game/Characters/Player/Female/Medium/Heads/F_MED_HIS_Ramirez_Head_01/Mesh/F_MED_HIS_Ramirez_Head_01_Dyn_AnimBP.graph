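# Post-process-style Animation Blueprint for the Ramirez female head, adding secondary motion to hair and ponytail bones over the input pose.
# Bool blends select between RigidBody, AnimDynamics, backbling-aware bone constraints, pass-through and shrunken-head poses, with a separate front-end RigidBody branch.
# Hat and helmet morph targets drive that branch's ponytail-to-head constraint alpha.
# An astronaut helmet check enables final neck and head rotation overrides.
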AnimBlueprint /Game/Characters/Player/Female/Medium/Heads/F_MED_HIS_Ramirez_Head_01/Mesh/F_MED_HIS_Ramirez_Head_01_Dyn_AnimBP
target_skeleton=/Game/Characters/Player/Common/Fortnite_Base_Head/Fortnite_Base_Head_Export_Skeleton
graphs [
 AnimGraph {nodes [
   n1 = "Constraint - dyn_hair_05_a (PP)"!(
    ComponentPose=Local To Component(LocalPose=Use cached pose 'PreScale Cache'())
    Alpha=1
    ConstraintWeights_0=0.5
    ConstraintWeights_1:float=0.5
   ) -> (Pose) {
    alphaBoolBlend {blendOption=Linear}
    boneToModify {boneName=dyn_hair_05_a}
    constraintSetup [{targetBone {boneName=neck_01}}, {targetBone {boneName=head}}]
    constraintWeights [1,1]
   }
   n2 = "Constraint - dyn_hair_05_b (PP)"!(ComponentPose=@n1, Alpha=1, ConstraintWeights_0=0.5, ConstraintWeights_1:float=0.5) -> (Pose) {
    alphaBoolBlend {blendOption=Linear}
    boneToModify {boneName=dyn_hair_05_b}
    constraintSetup [{targetBone {boneName=neck_01}}, {targetBone {boneName=head}}]
    constraintWeights [1,1]
   }
   n3 = "Constraint - dyn_hair_05_c (PP)"!(ComponentPose=@n2, Alpha=1, ConstraintWeights_0=0.5, ConstraintWeights_1:float=0.5) -> (Pose) {
    alphaBoolBlend {blendOption=Linear}
    boneToModify {boneName=dyn_hair_05_c}
    constraintSetup [{targetBone {boneName=spine_05}}, {targetBone {boneName=neck_01}}]
    constraintWeights [1,1]
   }
   n4 = RigidBody(ComponentPose=Local To Component(LocalPose=Use cached pose 'PreScale Cache'())) {
    bClampLinearTranslationLimitToRefPose=true
    bOverrideWorldGravity=true
    overridePhysicsAsset=&/F_MED_HIS_Ramirez_Head_01_Physics
    simulationSpace=BaseBoneSpace
    alphaBoolBlend {blendOption=Linear}
    baseBoneRef {boneName=root}
    componentAppliedLinearAccClamp=<500 500 1000>
    componentLinearAccScale=<0.2 0.3 1>
    componentLinearVelScale=<0.2 0.3 1>
    overrideWorldGravity {z=-980}
   }
   n5 = Blend Poses by bool(
    BlendPose_0=Component To Local(ComponentPose=@n3)
    BlendPose_1=Component To Local(ComponentPose=@n4)
    bActiveValue=Get Is Wearing Backbling()
   ) {blendType=Linear}
   n6 = Get HeadRotationBack() -> (HeadRotationBack:float)
   n7 = "Transform (Modify) Bone - Bone: dyn_hair_05"(
    ComponentPose=Local To Component(LocalPose=Use cached pose 'PreScale Cache'())
    Alpha=@n6
    Translation="0,0,0"
    Rotation="15,0.000,0.000"
    Scale="1,1,1"
   ) {
    rotationMode=BMM_Additive
    rotationSpace=BCS_BoneSpace
    translationMode=BMM_Additive
    translationSpace=BCS_BoneSpace
    alphaBoolBlend {blendOption=Linear}
    boneToModify {boneName=dyn_hair_05}
    rotation {pitch=15}
   }
   n8 = "Transform (Modify) Bone - Bone: dyn_hair_05_a"(ComponentPose=@n7, Alpha=@n6, Translation="0,0,0", Rotation="25,0.000,0.000", Scale="1,1,1") {
    rotationMode=BMM_Additive
    rotationSpace=BCS_BoneSpace
    translationMode=BMM_Additive
    translationSpace=BCS_BoneSpace
    alphaBoolBlend {blendOption=Linear}
    boneToModify {boneName=dyn_hair_05_a}
    rotation {pitch=25}
   }
   n9 = Anim Dynamics - Bone: dyn_hair_05_a(ComponentPose=@n8, Alpha=1, ExternalForce="0,0,0") {
    angularDampingOverride=0.9999
    angularSpringConstant=1500
    bGravityOverrideInSimSpace=true
    bOverrideAngularDamping=true
    bUseGravityOverride=true
    bUseSphericalLimits=true
    simulationSpace=RootRelative
    alphaBoolBlend {blendOption=Linear}
    boundBone {boneName=dyn_hair_05_a}
    componentAppliedLinearAccClamp=<1100 1100 1100>
    componentLinearAccScale=<0.5 0.5 0.5>
    componentLinearVelScale=<0.5 0.5 0.5>
    gravityOverride {z=-980}
    physicsBodyDefinitions [{
      collisionType=CustomSphere
      sphereCollisionRadius=2
      boundBone {boneName=dyn_hair_05_a}
      boxExtents=<8 15 15>
      constraintSetup {angularLimitsMax=<5 30 30>, angularLimitsMin=<-5 -30 -30>, angularTarget=<0 0 1>}
      localJointOffset=<5 0 0>
     }]
    sphericalLimits [{limitRadius=35, drivingBone {boneName=spine_05}, sphereLocalOffset=<-18 15 0>}]
   }
   n10 = Anim Dynamics - Bone: dyn_hair_05_b(ComponentPose=@n9, Alpha=1, ExternalForce="0,0,0") {
    angularDampingOverride=0.9999
    angularSpringConstant=1500
    bGravityOverrideInSimSpace=true
    bOverrideAngularDamping=true
    bUseGravityOverride=true
    bUseSphericalLimits=true
    simulationSpace=RootRelative
    alphaBoolBlend {blendOption=Linear}
    boundBone {boneName=dyn_hair_05_b}
    componentAppliedLinearAccClamp=<1100 1100 1100>
    componentLinearAccScale=<0.5 0.5 0.5>
    componentLinearVelScale=<0.5 0.5 0.5>
    gravityOverride {z=-980}
    physicsBodyDefinitions [{
      collisionType=CustomSphere
      sphereCollisionRadius=1.5
      boundBone {boneName=dyn_hair_05_b}
      boxExtents=<8 15 15>
      constraintSetup {angularLimitsMax=<-5 30 30>, angularLimitsMin=<-5 -30 -30>, angularTarget=<0 0 1>}
      localJointOffset=<2.5 0 0>
     }]
    sphericalLimits [{limitRadius=35, drivingBone {boneName=spine_05}, sphereLocalOffset=<-18 15 0>}]
   }
   n11 = Component To Local(ComponentPose=@n10)
   n12 = Get bRigidBodyEnabled()
   n13 = Anim Dynamics - Bone: dyn_hair_03(
    ComponentPose=Local To Component(LocalPose=Blend Poses by bool(BlendPose_0=@n5, BlendPose_1=@n11, bActiveValue=@n12) {blendType=Linear})
    Alpha=1
    ExternalForce="0,0,0"
   ) {
    angularDampingOverride=0.99
    angularSpringConstant=200
    bAngularSpring=true
    bGravityOverrideInSimSpace=true
    bOverrideAngularDamping=true
    alphaBoolBlend {blendOption=Linear}
    boundBone {boneName=dyn_hair_03}
    physicsBodyDefinitions [{
      sphereCollisionRadius=0
      boundBone {boneName=dyn_hair_03}
      boxExtents=<30 30 2>
      constraintSetup {angularLimitsMax=<5 15 0>, angularLimitsMin=<-15 1 0>, angularTarget=<0 0 1>}
      localJointOffset=<1.5 0 -10>
     }]
   }
   n14 = Anim Dynamics - Bone: dyn_hair_02(ComponentPose=@n13, Alpha=1) {
    angularDampingOverride=0.99
    angularSpringConstant=200
    bAngularSpring=true
    bGravityOverrideInSimSpace=true
    bOverrideAngularDamping=true
    alphaBoolBlend {blendOption=Linear}
    boundBone {boneName=dyn_hair_02}
    physicsBodyDefinitions [{
      sphereCollisionRadius=0
      boundBone {boneName=dyn_hair_02}
      boxExtents=<30 20 2>
      constraintSetup {angularLimitsMax=<5 5 0>, angularLimitsMin=<-15 -15 0>, angularTarget=<0 0 1>}
      localJointOffset=<1 0 -10>
     }]
   }
   n15 = Anim Dynamics - Bone: dyn_hair_04(ComponentPose=@n14, Alpha=0, ExternalForce="0,0,0") {
    angularSpringConstant=1500
    bGravityOverrideInSimSpace=true
    gravityScale=0.001
    numSolverIterationsPostUpdate=4
    numSolverIterationsPreUpdate=16
    alphaBoolBlend {blendOption=Linear}
    boundBone {boneName=dyn_hair_04}
    physicsBodyDefinitions [{
      sphereCollisionRadius=0
      boundBone {boneName=dyn_hair_04}
      boxExtents=<30 20 2>
      constraintSetup {
       angularTargetAxis=AxisY
       twistAxis=AxisY
       angularLimitsMax=<5 5 5>
       angularLimitsMin=<-5 -5 -5>
       angularTarget=<0 0 1>
       linearAxesMax=<0.5 0.5 0.5>
       linearAxesMin=<-0.5 -0.5 -0.5>
      }
      localJointOffset=<0 0 0>
     }]
   }
   n16 = "Transform (Modify) Bone - Bone: head"!(ComponentPose=Local To Component(LocalPose=Use cached pose 'PreScale Cache'()), Alpha=1) -> (Pose) {scaleMode=BMM_Replace, alphaBoolBlend {blendOption=Linear}, boneToModify {boneName=head}, scale=<0.01 0.01 0.01>}
   n17 = Blend Poses by bool(
    BlendPose_0=Component To Local(ComponentPose=@n16)
    BlendPose_1=Blend Poses by bool(
     BlendPose_0=Use cached pose 'Sub Graph Cache'()
     BlendPose_1=Component To Local(ComponentPose=@n15)
     bActiveValue=Get bCanPlayCustomAnimations() -> (bCanPlayCustomAnimations:bool)
    ) {blendType=Linear}
    BlendTime_0=0
    BlendTime_1=0
    bActiveValue=Get GetPlayerListUsesCustomCharacterParts() -> (GetPlayerListUsesCustomCharacterParts:bool)
   ) {blendType=Linear}
   n18 = RigidBody(ComponentPose=Local To Component(LocalPose=Use cached pose 'PreScale Cache'())) {
    bClampLinearTranslationLimitToRefPose=true
    bOverrideWorldGravity=true
    cachedBoundsScale=1.1
    overridePhysicsAsset=&/F_MED_HIS_Ramirez_Head_01_Physics1
    simulationSpace=BaseBoneSpace
    alphaBoolBlend {blendOption=Linear}
    baseBoneRef {boneName=head}
    componentAppliedLinearAccClamp=<100 100 100>
    componentLinearAccScale=<0.2 0.3 1>
    componentLinearVelScale=<0.2 0.3 1>
    overrideWorldGravity {z=-980}
   }
   n19 = Get HIddenPonytailAlpha() -> (HIddenPonytailAlpha:double)
   n20 = "Constraint - dyn_hair_05 (P)"(ComponentPose=@n18, Alpha=@n19, ConstraintWeights_0=1) {
    alphaBoolBlend {blendOption=Linear}
    boneToModify {boneName=dyn_hair_05}
    constraintSetup [{targetBone {boneName=head}}]
    constraintWeights [1]
   }
   n21 = "Constraint - dyn_hair_05_a (P)"(ComponentPose=@n20, Alpha=@n19, ConstraintWeights_0=1) {
    alphaBoolBlend {blendOption=Linear}
    boneToModify {boneName=dyn_hair_05_a}
    constraintSetup [{targetBone {boneName=head}}]
    constraintWeights [1]
   }
   n22 = "Constraint - dyn_hair_05_b (P)"(ComponentPose=@n21, Alpha=@n19, ConstraintWeights_0=1) {
    alphaBoolBlend {blendOption=Linear}
    boneToModify {boneName=dyn_hair_05_b}
    constraintSetup [{targetBone {boneName=head}}]
    constraintWeights [1]
   }
   n23 = "Constraint - dyn_hair_05_c (P)"(ComponentPose=@n22, Alpha=@n19, ConstraintWeights_0=1) {
    alphaBoolBlend {blendOption=Linear}
    boneToModify {boneName=dyn_hair_05_c}
    constraintSetup [{targetBone {boneName=head}}]
    constraintWeights [1]
   }
   n24 = Component To Local(ComponentPose=@n23)
   n25 = Get bIsInFrontEnd()
   n26 = Get Is helmet() -> (Is helmet:bool)
   n27 = "Transform (Modify) Bone - Bone: neck_02"!(
    ComponentPose=Local To Component(LocalPose=Blend Poses by bool(BlendPose_0=@n24, BlendPose_1=@n17, BlendTime_0=0, BlendTime_1=0, bActiveValue=@n25) {blendType=Linear})
    bAlphaBoolEnabled:bool=@n26
    Translation="0,0,0"
    Rotation="0,0,0"
    Scale="1,1,1"
   ) -> (Pose) {
    alphaInputType=Bool
    rotationMode=BMM_Replace
    rotationSpace=BCS_ParentBoneSpace
    alphaBoolBlend {blendOption=Linear}
    boneToModify {boneName=neck_02}
   }
   n28 = "Transform (Modify) Bone - Bone: head"!(ComponentPose=@n27, bAlphaBoolEnabled:bool=@n26, Translation:Vector="0,0,0", Rotation:Rotator="1.725,-11.263,1.388", Scale:Vector="1,1,1") -> (Pose) {
    alphaInputType=Bool
    rotationMode=BMM_Replace
    rotationSpace=BCS_ParentBoneSpace
    alphaBoolBlend {blendOption=Linear}
    boneToModify {boneName=head}
    rotation=<rot 1.725 -11.263 1.388>
   }
   Output Pose(Result=Component To Local(ComponentPose=@n28)) {layerGroup=None}
   Save cached pose 'PreScale Cache'(Pose=Use cached pose 'Sub Graph Cache'()) {cacheName=PreScale Cache}
   AnimGraphNode_SubInput_0 = Input Pose()
   Save cached pose 'Sub Graph Cache'(Pose=@AnimGraphNode_SubInput_0) {cacheName=Sub Graph Cache}
   GetOwnerSkeletalMeshComponent()
   Get HIddenPonytailAlpha()
  ]}
 EventGraph {nodes [
   Delay(Duration=0.1)
   > Do Once()
   Event BlueprintUpdateAnimation()
   n1 = Get MeshToCopy() -> (MeshToCopy:Skeletal Mesh Component Object Reference)
   n2 = "float > float"(A=GetMorphTarget!(self:Skeletal Mesh Component Object Reference=@n1, MorphTargetName:Name=Hair_Helmet) -> (ReturnValue:float), B=0.25)
   n3 = "float > float"(A=GetMorphTarget(self=@n1, MorphTargetName=Hair_Hat), B=0.25)
   n4 = "float > float"(A=GetMorphTarget(self=@n1, MorphTargetName=Hair_Hat), B=0.25)
   n5 = "float > float"(A=GetMorphTarget(self=@n1, MorphTargetName=Hair_SpaceSuit), B=0.25)
   > Set HIddenPonytailAlpha(HIddenPonytailAlpha:double=SelectFloat(A=0, B=1, bPickA=OR Boolean!(A=@n2, B=@n3, C:bool=@n4, D:bool=@n5) -> (ReturnValue))) -> (Output_Get:double)
   Event OnCharacterCustomizationLoaded()
   > n6 = Sequence!(execute) -> (then_0, then_1, then_2:Exec)
   n7 = TryGetPawnOwner()
   > Cast To PlayerPawn_Athena(Object=@n7)
   n8 = CheckForBackbling()
   Set Is Wearing Backbling(execute=@n6.then_1, Is Wearing Backbling=OR Boolean!(A=@n8, B=@n8.bIsBackblingLong) -> (ReturnValue))
   Is Valid(exec=@n6.then_2, InputObject=@n7)
   > n9 = Cast To FortPlayerPawnAthena(Object=@n7)
   n10 = GetSkeletalMeshForPartType(self=@n9, PartType=Hat)
   > Set Is helmet(Is helmet:bool="Equal (Object)"(
     A=GetSkeletalMeshAsset!(self:Skeletal Mesh Component Object Reference=@n10) -> (ReturnValue:Skeletal Mesh Object Reference)
     B=/Game/Characters/Player/Female/Medium/Bodies/F_MED_Astronaut_01/Meshes/Parts/F_Med_Astronaut_01_Helmet
    )) -> (Output_Get:bool)
   Event BlueprintInitializeAnimation()
   > Set MeshToCopy(MeshToCopy:Skeletal Mesh Component Object Reference=GetOwnerSkeletalMeshComponent()) -> (Output_Get:Skeletal Mesh Component Object Reference)
   > Set GetPlayerListUsesCustomCharacterParts(GetPlayerListUsesCustomCharacterParts:bool=GetPlaylistUsesCustomCharacterParts!() -> (ReturnValue:bool)) -> (Output_Get:bool)
   TryGetPawnOwner()
  ]}
]
variables [
 As Fort Player Pawn Athena {type=FortPlayerPawnAthena}
 GetPlayerListUsesCustomCharacterParts {type=boolean}
 HIddenPonytailAlpha {type=number}
 Is helmet {type=boolean}
 Is Wearing Backbling {type=boolean}
 MeshToCopy {type=SkeletalMeshComponent}
]
